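# Asset error scanner task that flags static meshes whose LODs do not reduce vertex count enough.
# checkDataForError reads the LODs asset tag and loads only assets with multiple LODs for checkAssetForError.
# That function compares each LOD's vertex count to the previous one, recording LODs above the similarityMax ratio or not lower at all.
# getTooltipAndColor turns both lists into the returned error's tooltip and severity color.
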EditorUtilityBlueprint /Game/Blueprints/Blutilities/ForTechArt/AssetErrorScanner/Tasks/Error_SM_LODSimilar_Task
graphs [
 EventGraph {links [], nodes []}
 checkAssetForError {nodes [
   n1 = checkAssetForError!() -> (then, asset:Object Reference) {is=title K2Node_FunctionEntry}
   > n2 = Cast To StaticMesh(Object=@n1)
   > Set l_mesh(l_mesh:Static Mesh Object Reference=@n2) -> (Output_Get:Static Mesh Object Reference)
   > n3 = Sequence()
   n4 = Get l_mesh() -> (l_mesh:Static Mesh Object Reference)
   > n5 = For Loop(LastIndex=int - int(A=GetNumLODs(self=@n4)))
   > n6 = Set l_LOD(l_LOD:int=@n5) -> (Output_Get:int=0)
   > n7 = Branch(Condition="Not Equal (Integer)"(A=@n6))
   > n8 = Sequence()
   > Set l_curVertCount(l_curVertCount:int=GetNumberVerts!(
     self:Static Mesh Editor Subsystem Object Reference=Get StaticMeshEditorSubsystem()
     StaticMesh:Static Mesh Object Reference=Get l_mesh()
     LODIndex:int=Get l_LOD() -> (l_LOD:int=0)
    ) -> (ReturnValue:int=0)) -> (Output_Get:int=0)
   > n9 = Branch(Condition="integer > integer"(A=Get l_lastVertCount() -> (l_lastVertCount:int=0), B=Get l_curVertCount() -> (l_curVertCount:int=0)))
   n10 = float * float(A:int=Get l_lastVertCount() -> (l_lastVertCount=0), B=Get similarityMax() -> (similarityMax:double)) {is=title K2Node_PromotableOperator}
   > Branch(Condition="integer > integer"(A=Get l_curVertCount() -> (l_curVertCount=0), B=Ceil(A=@n10) {is=title K2Node_CallFunction}))
   > Add(
    TargetArray:Array of Integers=Get LODsTooSimilar() -> (LODsTooSimilar:Array of Integers)
    NewItem:"Integer (by ref)"=Get l_LOD() -> (l_LOD=0)
   ) {is=title K2Node_CallArrayFunction}
   Add(execute=@n9.else, TargetArray=Get LODsMoreThanPrevious() -> (LODsMoreThanPrevious:Array of Integers), NewItem=Get l_LOD() -> (l_LOD=0))
   Set l_lastVertCount(execute=@n8.then_1, l_lastVertCount:int=Get l_curVertCount() -> (l_curVertCount=0)) -> (Output_Get:int=0)
   Set l_lastVertCount(
    execute=@n7.else
    l_lastVertCount=GetNumberVerts(self=Get StaticMeshEditorSubsystem(), StaticMesh=Get l_mesh(), LODIndex=Get l_LOD() -> (l_LOD=0)) -> (ReturnValue=0)
   ) -> (Output_Get=0)
   n11 = Length(TargetArray:Array of Integers=Get LODsTooSimilar())
   n12 = Length(TargetArray=Get LODsMoreThanPrevious())
   Branch(execute=@n3.then_1, Condition="integer > integer"(A=int + int(A=@n12, B=@n11)))
   n13 = getTooltipAndColor!() -> (color:Linear Color Structure, ReturnValue:String)
   > Return Node!(
    execute
    isError:bool=true
    error:AVT Error Struct Structure=Make AVT Error Struct!(
     title_32_#1:String=LODs
     description_35_#2:String=LODs Too Similar
     tooltip_37_#3:String=@n13.ReturnValue
     color_40_#4:Linear Color Structure=@n13
     csvLine_46_#5:String=Yes
     csvHeader_49_#6:String=LODs Too Similar
     optionalAssetData_52_#7:Asset Data Structure=Get assetData() -> (assetData:Asset Data Structure)
     optionalTypeTag_55_#8:String
     optionalOriginPackageName_62_#9:Name=None
     helpCommand_65_#10:String
    ) -> (AVT_Error_Struct:AVT Error Struct Structure) {is=title K2Node_MakeStruct}
   )
  ]}
 checkDataForError {nodes [
   checkDataForError!() -> (then, assetData:Asset Data Structure) {is=title K2Node_FunctionEntry}
   n1 = GetTagValue(InAssetData=Get assetData(), InTagName=MakeLiteralName(Value=LODs))
   > Branch(Condition=@n1.ReturnValue)
   > n2 = Set LODCount(LODCount:int=String To Integer(InString=@n1)) -> (Output_Get:int=0)
   > Branch(Condition="integer > integer"(A=@n2, B=1))
   > n3 = checkAssetForError!(execute, self:Base Error Task Object Reference, asset:Object Reference=GetAsset(InAssetData=Get assetData())) -> (then, isError:bool, error:AVT Error Struct Structure) {is=title K2Node_CallFunction}
   > Return Node(isError=@n3, error=@n3.error)
  ]}
 getTooltipAndColor {nodes [
   getTooltipAndColor() {is=title K2Node_FunctionEntry}
   > n1 = Sequence!(execute) -> (then_0, then_1, then_2:Exec)
   n2 = Get LODsTooSimilar()
   > Branch(Condition="integer > integer"(A=Length(TargetArray=@n2)))
   > Set l_error(l_error:String="These LODs are not different enough from their previous LOD: ") -> (Output_Get:String)
   n3 = Get LODsTooSimilar()
   > n4 = For Each Loop(Array:Array of Integers=@n3) -> (Array Element:int)
   Set l_color(execute=@n4.Completed, l_color:Linear Color Structure=getErrorColor!(badness:double=0.85) -> (Color:Linear Color Structure)) -> (Output_Get:Linear Color Structure)
   n5 = "Not Equal (Integer)"(A=@n4.Array Index, B=int - int(A=Length(TargetArray=@n3)))
   Set l_error(
    execute=@n4.LoopBody
    l_error=Append!(A=Get l_error() -> (l_error:String), B="To String (Integer)"(InInt=@n4), C:String=SelectString(A=", ", B=., bPickA=@n5)) -> (ReturnValue)
   )
   n6 = Get LODsMoreThanPrevious()
   Branch(execute=@n1.then_1, Condition="integer > integer"(A=Length(TargetArray=@n6)))
   > Set l_error(l_error=Append!(A=Get l_error(), B="\r\nThese LODs are more than their previous LOD: ") -> (ReturnValue))
   n7 = Get LODsMoreThanPrevious()
   > n8 = For Each Loop(Array=@n7)
   Set l_color(execute=@n8.Completed, l_color=getErrorColor(badness=1))
   n9 = "Not Equal (Integer)"(A=@n8.Array Index, B=int - int(A=Length(TargetArray=@n7)))
   Set l_error(
    execute=@n8.LoopBody
    l_error=Append!(A=Get l_error(), B="To String (Integer)"(InInt=@n8), C:String=SelectString(A=", ", B=., bPickA=@n9)) -> (ReturnValue)
   )
   n10 = Get!() -> (assetData_PackageName:Name, assetData_PackagePath:Name, assetData_AssetName:Name, assetData_AssetClassPath:Top Level Asset Path Structure)
   Return Node!(
    execute=@n1.then_2
    color:Linear Color Structure=Get l_color() -> (l_color:Linear Color Structure)
    ReturnValue:String=Append!(A="To String (Name)"(InName=@n10.assetData_AssetName), B=" has improper LODs. ", C="\r\n", D:String=Get l_error()) -> (ReturnValue)
   )
  ]}
]
variables [
 LODCount {type=integer}
 LODsMoreThanPrevious {type="array<integer>"}
 LODsTooSimilar {type="array<integer>"}
 similarityMax {type=number, default=0.75}
]
