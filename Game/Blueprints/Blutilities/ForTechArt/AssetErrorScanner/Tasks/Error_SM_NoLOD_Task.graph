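# Editor utility task for the Asset Error Scanner that flags static mesh LOD problems from asset data.
# It reads LODs and Vertices tags, flagging single-LOD meshes above a vertex threshold unless the shouldRunType set indicates FX-only.
# For FX-only multi-LOD meshes, it loads the asset and flags it when the highest per-quality MinLOD falls short of the last LOD.
# Each failure returns an error struct with tooltip, color and CSV fields.
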EditorUtilityBlueprint /Game/Blueprints/Blutilities/ForTechArt/AssetErrorScanner/Tasks/Error_SM_NoLOD_Task
graphs [
 EventGraph {links [], nodes []}
 checkAssetForError {nodes [
   checkAssetForError!() -> (then, asset:Object Reference) {is=title K2Node_FunctionEntry}
   > n1 = checkDataForError!(execute, self:Base Error Task Object Reference, assetData:Asset Data Structure=Get assetData() -> (assetData:Asset Data Structure)) -> (then, isError:bool, error:AVT Error Struct Structure)
   > Return Node!(execute, isError:bool=@n1, error:AVT Error Struct Structure=@n1.error)
  ]}
 checkDataForError {nodes [
   checkDataForError!() -> (then, assetData:Asset Data Structure) {is=title K2Node_FunctionEntry}
   > validateFXAsset!(execute, self:Error FXOnly Base Task Object Reference) -> (then, shouldRun:bool)
   n1 = GetTagValue(InAssetData=Get assetData(), InTagName=MakeLiteralName(Value=LODs))
   n2 = GetTagValue(InAssetData=Get assetData(), InTagName=MakeLiteralName(Value=Vertices))
   > Branch(Condition=AND Boolean(A=@n1.ReturnValue, B=@n2.ReturnValue))
   > Set l_vertCount(l_vertCount:int=String To Integer(InString=@n2)) -> (Output_Get:int=0)
   > n3 = Set l_lodCount(l_lodCount:int=String To Integer(InString=@n1)) -> (Output_Get:int=0)
   > n4 = Branch(Condition="Equal (Integer)"(A=@n3, B=1))
   > Branch(Condition="integer > integer"(A=Get l_vertCount() -> (l_vertCount:int=0), B=Get minVertsForError() -> (minVertsForError:int=0)))
   > Branch(Condition="Not Equal (Enum)"!(A:AES_FXMesh_Enum Enum=Get shouldRunType() -> (shouldRunType:AES_FXMesh_Enum Enum=NewEnumerator0), B:AES_FXMesh_Enum Enum=NewEnumerator1) -> (ReturnValue:bool))
   n5 = "To String (Name)"(InName=Get!() -> (assetData_PackageName:Name, assetData_PackagePath:Name, assetData_AssetName:Name, assetData_AssetClassPath:Top Level Asset Path Structure).assetData_AssetName)
   n6 = "To String (Integer)"(InInt=Get minVertsForError() -> (minVertsForError=0))
   n7 = "To String (Integer)"(InInt=Get l_vertCount() -> (l_vertCount=0))
   > Return Node(
    isError=true
    error=Make AVT Error Struct!(
     title_32_#1:String=LODs
     description_35_#2:String=No LODs
     tooltip_37_#3:String=Append!(A=@n5, B=" has no LODs and is more than ", C:String=@n6, D:String=" verts (", E:String=@n7, F:String=")") -> (ReturnValue)
     color_40_#4:Linear Color Structure=getErrorColor!(badness:double=1) -> (Color:Linear Color Structure)
     csvLine_46_#5:String=Yes
     csvHeader_49_#6:String=No LODs
     optionalAssetData_52_#7:Asset Data Structure=Get assetData()
     optionalTypeTag_55_#8:String
     optionalOriginPackageName_62_#9:Name=None
     helpCommand_65_#10:String
    ) -> (AVT_Error_Struct:AVT Error Struct Structure) {is=title K2Node_MakeStruct}
   )
   Branch(
    execute=@n4.else
    Condition="Equal (Enum)"(A:AES_FXMesh_Enum Enum=Get shouldRunType() -> (shouldRunType=NewEnumerator0), B:AES_FXMesh_Enum Enum=NewEnumerator1)
   )
   > n8 = Sequence()
   > n9 = Cast To StaticMesh(Object=GetAsset(InAssetData=Get assetData()))
   > Set l_maxLODOverridden(l_maxLODOverridden:int=0) -> (Output_Get:int=0)
   > n10 = Values!(
    execute
    TargetMap:Map of Names to Integers=GetMinimumLODForQualityLevels!(self:Static Mesh Object Reference=@n9) -> (QualityLevelMinimumLODs:Map of Names to Integers)
   ) -> (then, Values:Array of Integers)
   > n11 = For Each Loop with Break!(Exec:Exec, Array:Array of Integers=@n10, Break:Exec) -> (LoopBody:Exec, Array Element:int, Array Index:int, Completed:Exec)
   n12 = Get l_lodCount() -> (l_lodCount:int=0)
   n13 = Get l_maxLODOverridden() -> (l_maxLODOverridden:int=0)
   Set l_hasMinLODForAllLODs(execute=@n11.Completed, l_hasMinLODForAllLODs:bool="integer <= integer"(A=int - int(A=@n12), B=@n13)) -> (Output_Get:bool)
   Set l_maxLODOverridden(execute=@n11.LoopBody, l_maxLODOverridden="Max (Integer)"(A=@n11, B=@n13)) -> (Output_Get=0)
   n14 = Branch(execute=@n8.then_1, Condition=Get l_hasMinLODForAllLODs() -> (l_hasMinLODForAllLODs:bool))
   n15 = "To String (Name)"(InName=Get().assetData_AssetName)
   Return Node(
    execute=@n14.else
    isError=true
    error=Make AVT Error Struct(
     title_32_#1=LODs
     description_35_#2=FX With LODs
     tooltip_37_#3=Append!(
      A=@n15
      B=" hasLODs and is only referenced by Niagara which does not use LODs and at least one LOD is not used as a MinLOD.  Remove unneeded LODs or use all LODs as MinLODs."
     ) -> (ReturnValue)
     color_40_#4=getErrorColor(badness=1)
     csvLine_46_#5=Yes
     csvHeader_49_#6=FX With LODs
     optionalAssetData_52_#7=Get assetData()
     optionalTypeTag_55_#8=FX
     optionalOriginPackageName_62_#9=None
    )
   )
  ]}
]
variables [
 LODCount {type=integer}
 LODsMoreThanPrevious {type="array<integer>"}
 LODsTooSimilar {type="array<integer>"}
 minVertsForError {type=integer, default=100}
 similarityMax {type=number, default=0.75}
]
