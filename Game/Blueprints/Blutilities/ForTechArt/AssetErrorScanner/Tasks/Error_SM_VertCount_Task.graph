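# Editor utility task for the asset error scanner that flags static meshes with too many vertices.
# It reads the Vertices asset-registry tag and compares it to a general limit or a lower FX limit, chosen by isFXAsset after validateFXAsset gates the check.
# Over-limit meshes return an error struct with count text, FX tagging, CSV fields and a color lerped from orange toward red by the count-to-limit ratio.
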EditorUtilityBlueprint /Game/Blueprints/Blutilities/ForTechArt/AssetErrorScanner/Tasks/Error_SM_VertCount_Task
graphs [
 EventGraph {links [], nodes []}
 checkAssetForError {nodes [
   n1 = checkAssetForError!() -> (then, asset:Object Reference) {is=title K2Node_FunctionEntry}
   > Cast To StaticMesh(Object=@n1)
   > n2 = checkDataForError!(execute, self:Base Error Task Object Reference, assetData:Asset Data Structure=Get assetData() -> (assetData:Asset Data Structure)) -> (then, isError:bool, error:AVT Error Struct Structure)
   > Return Node!(execute, isError:bool=@n2, error:AVT Error Struct Structure=@n2.error)
  ]}
 checkDataForError {nodes [
   checkDataForError!() -> (then, assetData:Asset Data Structure) {is=title K2Node_FunctionEntry}
   > n1 = validateFXAsset!(execute, self:Error FXOnly Base Task Object Reference) -> (then, shouldRun:bool)
   > Branch(Condition=@n1)
   n2 = GetTagValue(InAssetData=Get assetData(), InTagName=MakeLiteralName(Value=Vertices))
   > n3 = Branch(Condition=@n2.ReturnValue)
   > n4 = Set VertexCount(VertexCount:int=String To Integer(InString=@n2)) -> (Output_Get:int=0)
   > n5 = Branch(Condition="integer > integer"(A=@n4, B=getValidatedMaxVertexCount!() -> (maxVertexCount:int=0)))
   n6 = Get VertexCount() -> (VertexCount:int=0)
   n7 = float / float(A:int=@n6, B:int=getValidatedMaxVertexCount() -> (maxVertexCount=0))
   n8 = Get isFXAsset() -> (isFXAsset:bool)
   n9 = SelectString(
    A=" vertices.  This is too high for an FX mesh (which this seems to be)"
    B=" vertices.  This is too high for an average model."
    bPickA=Get isFXAsset()
   )
   n10 = "To String (Integer)"(InInt=@n6)
   n11 = "To String (Name)"(InName=Get!() -> (assetData_PackageName:Name, assetData_PackagePath:Name, assetData_AssetName:Name, assetData_AssetClassPath:Top Level Asset Path Structure).assetData_AssetName)
   n12 = SelectString(A="FX - ", bPickA=Get isFXAsset())
   > Return Node(
    isError=true
    error=Make AVT Error Struct!(
     title_32_#1:String=Vertices
     description_35_#2:String=Append!(A=@n12, B=@n10, C:String=" Verts") -> (ReturnValue)
     tooltip_37_#3:String=Append!(A=@n11, B=" has ", C=@n10, D:String=@n9) -> (ReturnValue)
     color_40_#4:Linear Color Structure="Lerp (LinearColor)"(A=(R=1,G=0.5265,B=0.1983,A=1), B=(R=1,G=0,B=0,A=1), Alpha=@n7)
     csvLine_46_#5:String=Yes
     csvHeader_49_#6:String=Too Many Verts
     optionalAssetData_52_#7:Asset Data Structure=Get assetData()
     optionalTypeTag_55_#8:String=SelectString(A=FX, bPickA=@n8)
     optionalOriginPackageName_62_#9:Name=None
     helpCommand_65_#10:String
    ) -> (AVT_Error_Struct:AVT Error Struct Structure) {is=title K2Node_MakeStruct}
   )
   Return Node(execute=[@n5.else, @n3.else], isError=false)
   "To String (Integer)"(InInt=Get VertexCount() -> (VertexCount=0))
  ]}
 getValidatedMaxVertexCount {nodes [
   getValidatedMaxVertexCount() {is=title K2Node_FunctionEntry}
   > Return Node!(
    execute
    maxVertexCount:int=Select(
     Option 0:int=Get maxVertexCount() -> (maxVertexCount:int=0)
     Option 1:int=Get maxVertexCount_FX() -> (maxVertexCount_FX:int=0)
     Index=Get isFXAsset()
    ) -> (ReturnValue:int)
   )
  ]}
]
variables [
 maxVertexCount {type=integer, default=10000}
 maxVertexCount_FX {type=integer, default=2000}
 VertexCount {type=integer}
]
