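# Editor utility widget for tech art that runs asset error scans and lists the failing assets.
# RunTests launches an AssetScanForError task with a saved skip map, shows step progress, then fills the list from the task's final errors.
# Closing an error can, after confirmation, save that test as skipped for the asset, while help executes the error's Python command.
# PrintErrorDataCsv prints errors as CSV-style rows with vertex counts from Asset Registry tags.
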EditorUtilityWidgetBlueprint /Game/Blueprints/Blutilities/ForTechArt/AssetErrorScanner/Widgets/AssetErrorList_EWidget
graphs [
 AddAssetToList {nodes [
   n1 = AddAssetToList!() -> (then, assetErrorData:AES Test Result Struct Structure, displayString:String, packageName:String) {is=title K2Node_FunctionEntry}
   > Set l_aesErrorData(l_aesErrorData:AES Test Result Struct Structure=@n1) -> (Output_Get:AES Test Result Struct Structure)
   > n2 = Sequence()
   > Set members in AVT Asset Data Struct!(
    execute:Exec
    StructRef:"AVT Asset Data Struct Structure (by ref)"=Get l_widgetErrorData() -> (l_widgetErrorData:AVT Asset Data Struct Structure)
    errors_35_#1:Array of AVT Error Struct Structures=Get!() -> (
     l_aesErrorData_assetPackageName_15_#2:Name
     l_aesErrorData_errors_17_#3:Array of AVT Error Struct Structures
    ).l_aesErrorData_errors_17_#3
    assetPackageName_46_#4:String=@n1.packageName
   ) -> (then:Exec, StructOut:"AVT Asset Data Struct Structure (by ref)") {is=title K2Node_SetFieldsInStruct}
   n3 = Construct Object from Class(
    execute=@n2.then_1
    Class=/Game/Blueprints/Blutilities/ForArtists/AssetValidationTool/Data/AVT_ItemData.AVT_ItemData_C
    self=Self-Reference()
   ) -> (ReturnValue:AVT Item Data Object Reference)
   > Set displayString!(execute, displayString:String=@n1.displayString, self:TA Generic List Item Data Object Reference=@n3) -> (then, Output_Get:String)
   > Set assetErrorData!(execute, assetErrorData:AVT Asset Data Struct Structure=Get l_widgetErrorData(), self:AVT Item Data Object Reference=@n3) -> (then, Output_Get:AVT Asset Data Struct Structure)
   > Set strings!(
    execute
    strings:Array of Strings=Make Array("[0]":String=@n1.packageName) -> (Array:Array of Strings)
    self:TA Generic List Item Data Object Reference=@n3
   ) -> (then, Output_Get:Array of Strings)
   > n4 = addItem!(
    execute
    self:TA Generic 01 List Object Reference=Get TA_Generic01_List() -> (TA_Generic01_List:TA Generic 01 List Object Reference)
    dataObject:Object Reference=@n3
   ) -> (then, newItem:User Widget Object Reference, newData:Object Reference)
   > n5 = Cast To AVT_Asset_ListItemWidget!(execute, Object=@n4) -> (then, CastFailed, AsAVT Asset List Item Widget:AVT Asset List Item Widget Object Reference) {is=title K2Node_DynamicCast}
   > Bind Event to On Error Close Clicked(self:AVT Asset List Item Widget Object Reference=@n5, Delegate=Create Event()) {is=title K2Node_AddDelegate}
   > Bind Event to On Error Help Clicked(self:AVT Asset List Item Widget Object Reference=@n5, Delegate=Create Event()) {is=title K2Node_AddDelegate}
  ]}
 EventGraph {nodes [
   n1 = RunTests!() -> (OutputDelegate, then, assetTests:Array of AES Tests Struct Structures) {is=title K2Node_CustomEvent}
   > Branch(Condition=Is Valid Index!(TargetArray:Array of AES Tests Struct Structures=@n1.assetTests, IndexToTest:int) -> (ReturnValue:bool))
   > n2 = Sequence()
   > n3 = Cast To AES_SaveGame!(execute, Object=Get saveGame_Object() -> (saveGame_Object:Save Game Object Reference)) -> (then, CastFailed, AsAES Save Game:AES Save Game Object Reference) {is=title K2Node_DynamicCast}
   > Set currentSkippedAssetTestMap(currentSkippedAssetTestMap:Map of Names to AES Skip Test Struct Structures=Get skippedAssetTestMap!(self:AES Save Game Object Reference=@n3) -> (skippedAssetTestMap:Map of Names to AES Skip Test Struct Structures)) -> (Output_Get:Map of Names to AES Skip Test Struct Structures)
   n4 = Branch(execute=@n2.then_1, Condition=Get isScanning() -> (isScanning:bool))
   cancelScan(execute=@n4.else)
   > setScanMode!(execute, self, isScanning:bool=true) -> (then)
   > clearItems(self:TA Generic 01 List Object Reference=Get TA_Generic01_List())
   > n5 = Construct Object from Class!(
    execute
    Class=/Game/Blueprints/Blutilities/ForTechArt/AssetErrorScanner/Tasks/AssetScanForError_Task.AssetScanForError_Task_C
    self
    assetTests:Array of AES Tests Struct Structures=@n1.assetTests
    assetErrorsToSkip:Map of Names to AES Skip Test Struct Structures=Get currentSkippedAssetTestMap() -> (currentSkippedAssetTestMap:Map of Names to AES Skip Test Struct Structures)
   ) -> (then, ReturnValue:Asset Scan For Error Task Object Reference)
   > n6 = Set ScanTask(ScanTask:Asset Scan For Error Task Object Reference=@n5) -> (Output_Get:Asset Scan For Error Task Object Reference)
   > Bind Event to On Test Step Complete(self:Asset Scan For Error Task Object Reference=@n6, Delegate=Create Event()) {is=title K2Node_AddDelegate}
   n7 = Get ScanTask() -> (ScanTask:Asset Scan For Error Task Object Reference)
   > n8 = RegisterAndExecuteTask!(execute:Exec, Task:Editor Utility Task Object Reference=@n7, OptionalParentTask:Editor Utility Task Object Reference) -> (then:Exec, OnFinished:Exec, Task:Editor Utility Task Object Reference) {is=title K2Node_AsyncAction}
   n9 = Values!(
    execute=@n8.OnFinished
    TargetMap:Map of Names to AES Test Result Struct Structures=Get finalErrors!(self:Asset Scan For Error Task Object Reference=@n7) -> (finalErrors:Map of Names to AES Test Result Struct Structures)
   ) -> (then, Values:Array of AES Test Result Struct Structures)
   > n10 = For Each Loop(Array:Array of AES Test Result Struct Structures=@n9) -> (Array Element:AES Test Result Struct Structure)
   Set ScanTask(execute=@n10.Completed)
   > setScanMode(isScanning=false)
   n11 = "To String (Name)"(InName=Break AES Test Result Struct!(AES_TestResult_Struct:"AES Test Result Struct Structure (by ref)"=@n10) -> (assetPackageName_15_#2:Name, errors_17_#3:Array of AVT Error Struct Structures) {is=title K2Node_BreakStruct})
   AddAssetToList!(
    execute=@n10.LoopBody
    self
    assetErrorData:AES Test Result Struct Structure=@n10
    displayString:String=Split(SourceString=@n11, InStr=/, SearchDir=FromEnd).RightS
    packageName:String=@n11
   ) -> (then) {is=title K2Node_CallFunction}
   Event Construct()
   > Parent: Construct()
  ]}
 HandleItemErrorCloseClicked {nodes [
   n1 = HandleItemErrorCloseClicked!() -> (then, errorWidget:AVT Asset Error Widget Object Reference, itemData:AVT Item Data Object Reference) {is=title K2Node_FunctionEntry}
   > Set l_errorWidget(l_errorWidget:AVT Asset Error Widget Object Reference=@n1) -> (Output_Get:AVT Asset Error Widget Object Reference)
   n2 = Break AVT Error Struct!(AVT_Error_Struct:"AVT Error Struct Structure (by ref)"=Get currentError!(self:AVT Asset Error Widget Object Reference=Get l_errorWidget() -> (l_errorWidget:AVT Asset Error Widget Object Reference)) -> (currentError:AVT Error Struct Structure)) -> (
    optionalAssetData_52_#5:Asset Data Structure
    optionalOriginPackageName_62_#6:Name
    helpCommand_65_#7:String
   ) {is=title K2Node_BreakStruct}
   > Set l_errorTaskName(l_errorTaskName:Name=@n2.optionalOriginPackageName_62_#6) -> (Output_Get:Name=None)
   > Set l_assetPackageName(l_assetPackageName:Name=Break Asset Data(AssetData=@n2)) -> (Output_Get:Name=None)
   n3 = Break AVT Error Struct!(AVT_Error_Struct=Get currentError(self=Get l_errorWidget())) -> (
    title_32_#8:String
    description_35_#9:String
    optionalAssetData_52_#5
    helpCommand_65_#7
   )
   n4 = "To String (Name)"(InName=Break Asset Data!(AssetData=@n3.optionalAssetData_52_#5) -> (AssetName, AssetClassPath))
   n5 = Append!(A=@n3, B=": ", C:String=@n3.description_35_#9) -> (ReturnValue)
   > n6 = Show Message Dialog(
    Title="To Text (String)"(InString=Append(A="Remove error: ", B=@n3, C=?))
    Message="To Text (String)"(InString=Append!(
      A="Clicking Yes will make this error never show again for this asset.\r\n\r\nNever show \r\n"
      B=@n5
      C="\r\nfor\r\n"
      D:String=@n4
      E:String="\r\nagain?"
     ) -> (ReturnValue))
    MessageType=YesNo
    DefaultValue=No
    MessageCategory=Warning
   )
   > n7 = Switch on EAppReturnType(Selection=@n6)
   n8 = Sequence(execute=@n7.Yes)
   > n9 = Cast To AES_SaveGame(Object=Get saveGame_Object())
   > n10 = Set currentSkippedAssetTestMap(currentSkippedAssetTestMap=Get skippedAssetTestMap(self=@n9))
   n11 = Find(
    TargetMap:Map of Names to AES Skip Test Struct Structures=Get currentSkippedAssetTestMap()
    Key:"Name (by ref)"=Get l_assetPackageName() -> (l_assetPackageName:Name=None)
   ) -> (Value:AES Skip Test Struct Structure)
   n12 = Branch(execute=[@n9.CastFailed, @n10.then], Condition=@n11.ReturnValue)
   > n13 = Set l_error(l_error:AES Skip Test Struct Structure=@n11) -> (Output_Get:AES Skip Test Struct Structure)
   n14 = Break AES Skip Test Struct!(AES_SkipTest_Struct:"AES Skip Test Struct Structure (by ref)"=@n13) -> (assetPackageName_14_#2:Name, skippedTests_15_#3:Array of Names) {is=title K2Node_BreakStruct}
   > n15 = Set l_tests(l_tests:Array of Names=@n14.skippedTests_15_#3) -> (Output_Get:Array of Names)
   > Add(TargetArray:Array of Names=@n15, NewItem:"Name (by ref)"=Get l_errorTaskName() -> (l_errorTaskName:Name=None)) {is=title K2Node_CallArrayFunction}
   > n16 = Set l_error(l_error=Make AES Skip Test Struct!(
     assetPackageName_14_#2:Name=@n14
     skippedTests_15_#3:Array of Names=Get l_tests() -> (l_tests:Array of Names)
    ) -> (AES_SkipTest_Struct:AES Skip Test Struct Structure) {is=title K2Node_MakeStruct})
   > Add!(
    execute
    TargetMap:Map of Names to AES Skip Test Struct Structures=Get currentSkippedAssetTestMap()
    Key:"Name (by ref)"=Get l_assetPackageName() -> (l_assetPackageName=None)
    Value:"AES Skip Test Struct Structure (by ref)"=@n16
   ) -> (then) {is=title K2Node_CallFunction}
   n17 = Get l_errorTaskName() -> (l_errorTaskName=None)
   Add(
    execute=@n12.else
    TargetMap=Get currentSkippedAssetTestMap()
    Key=Get l_assetPackageName() -> (l_assetPackageName=None)
    Value=Make AES Skip Test Struct(
     assetPackageName_14_#2=Get l_assetPackageName() -> (l_assetPackageName=None)
     skippedTests_15_#3=Make Array("[0]":Name=@n17) -> (Array:Array of Names)
    )
   )
   SaveGame_Save(execute=@n8.then_1, self:TA Base Editor Widget Object Reference)
   > removeSelf(self:AVT Asset Error Widget Object Reference=Get l_errorWidget())
  ]}
 HandleStepComplete {nodes [
   n1 = HandleStepComplete!() -> (then, EventName:String, stepIndex:int, stepTotal:int) {is=title K2Node_FunctionEntry}
   > setProgress!(
    execute
    self:AES Progress Widget Object Reference=Get AES_ProgressWidget() -> (AES_ProgressWidget:AES Progress Widget Object Reference)
    progressString:String=@n1
    percent:double=float / float(A:int=@n1.stepIndex, B:int=@n1.stepTotal)
   ) -> (then)
  ]}
 PrintErrorDataCsv {nodes [
   PrintErrorDataCsv() {is=title K2Node_FunctionEntry}
   > n1 = Sequence!(execute) -> (then_0, then_1, then_2:Exec, then_3:Exec, then_4:Exec)
   > Add!(execute, TargetSet:Set of Strings=Get l_headers() -> (l_headers:Set of Strings), NewItem:"String (by ref)"=Asset Name) -> (then)
   n2 = getListData!(execute=@n1.then_1, self:TA Generic 01 List Object Reference=Get TA_Generic01_List()) -> (then, data:Array of Object References)
   > n3 = Branch(Condition=Is Valid Index(TargetArray:Array of Object References=@n2))
   > n4 = For Each Loop(Array:Array of Object References=@n2) -> (Array Element:Object Reference)
   > n5 = Cast To AVT_ItemData!(execute, Object=@n4) -> (then, CastFailed, AsAVT Item Data:AVT Item Data Object Reference) {is=title K2Node_DynamicCast}
   n6 = Get!(self:AVT Item Data Object Reference=@n5) -> (
    assetErrorData_asset_39_#10:Object Reference
    assetErrorData_errors_35_#1:Array of AVT Error Struct Structures
    assetErrorData_csvLine_42_#11:String
    assetErrorData_assetPackageName_46_#4:String
   )
   > Append Array!(
    execute
    TargetArray:Array of AVT Error Struct Structures=Get l_allErrors() -> (l_allErrors:Array of AVT Error Struct Structures)
    SourceArray:Array of AVT Error Struct Structures=@n6.assetErrorData_errors_35_#1
   ) -> (then)
   > n7 = For Each Loop(Array:Array of AVT Error Struct Structures=@n6.assetErrorData_errors_35_#1) -> (Array Element:AVT Error Struct Structure)
   > Add(
    TargetSet=Get l_headers()
    NewItem=Break AVT Error Struct!(AVT_Error_Struct=@n7) -> (
     title_32_#8
     description_35_#9
     tooltip_37_#12:String
     color_40_#13:Linear Color Structure
     csvLine_46_#14:String
     csvHeader_49_#15:String
     optionalAssetData_52_#5
     optionalTypeTag_55_#16:String
     optionalOriginPackageName_62_#6:Name
     helpCommand_65_#7
    ).csvHeader_49_#15
   )
   Return Node(execute=@n3.else)
   n8 = For Each Loop(Exec=@n1.then_2, Array=Get l_allErrors())
   > Set l_error(l_error:AVT Error Struct Structure=@n8) -> (Output_Get:AVT Error Struct Structure)
   n9 = Break AVT Error Struct(AVT_Error_Struct=@n8)
   > Set l_assetData(l_assetData=@n9.optionalAssetData_52_#5)
   > n10 = Set l_assetName(l_assetName:String="To String (Name)"(InName=Break Asset Data!(AssetData=@n9.optionalAssetData_52_#5) -> (PackageName:Name, AssetClassPath))) -> (Output_Get:String)
   n11 = Find(
    TargetMap:Map of Strings to AES CSVOutput Struct Structures=Get l_AssetColumnMap() -> (l_AssetColumnMap:Map of Strings to AES CSVOutput Struct Structures)
    Key:"String (by ref)"=@n10
   ) -> (Value:AES CSVOutput Struct Structure)
   > n12 = Branch(Condition=@n11.ReturnValue)
   > n13 = Set l_tempCsvOuput(l_tempCsvOuput:AES CSVOutput Struct Structure=@n11) -> (Output_Get:AES CSVOutput Struct Structure)
   > n14 = Set l_tempColumns(l_tempColumns:Map of Strings to Strings=Break AES CSVOutput Struct!(AES_CSVOutput_Struct:"AES CSVOutput Struct Structure (by ref)"=@n13) -> (
     rowName_6_#17:String
     columns_9_#18:Map of Strings to Strings
     tags_17_#19:Array of Strings
     vertcount_14_#20:int
    ) {is=title K2Node_BreakStruct}.columns_9_#18) -> (Output_Get:Map of Strings to Strings)
   n15 = Break AVT Error Struct!(AVT_Error_Struct=Get l_error() -> (l_error:AVT Error Struct Structure)) -> (
    csvLine_46_#14
    csvHeader_49_#15
    optionalOriginPackageName_62_#6
    helpCommand_65_#7
   )
   > Add!(
    execute
    TargetMap:Map of Strings to Strings=@n14
    Key:"String (by ref)"=@n15.csvHeader_49_#15
    Value:"String (by ref)"=@n15
   ) -> (then)
   > Set members in AES CSVOutput Struct!(
    execute:Exec
    StructRef:"AES CSVOutput Struct Structure (by ref)"=Get l_tempCsvOuput() -> (l_tempCsvOuput:AES CSVOutput Struct Structure)
    columns_9_#18:Map of Strings to Strings=Get l_tempColumns() -> (l_tempColumns:Map of Strings to Strings)
   ) -> (then:Exec, StructOut:"AES CSVOutput Struct Structure (by ref)") {is=title K2Node_SetFieldsInStruct}
   > Add(
    TargetMap:Map of Strings to AES CSVOutput Struct Structures=Get l_AssetColumnMap()
    Key=Get l_assetName() -> (l_assetName:String)
    Value:"AES CSVOutput Struct Structure (by ref)"=Get l_tempCsvOuput()
   )
   Clear!(execute=@n12.else, TargetMap:Map of Strings to Strings=Get l_tempColumns()) -> (then) {is=title K2Node_CallFunction}
   > n16 = To Array!(execute, A:Set of Strings=Get l_headers()) -> (then, Result:Array of Strings)
   > n17 = For Each Loop(Array:Array of Strings=@n16) -> (Array Element:String)
   Add(execute=@n17.Completed, TargetMap:Map of Strings to Strings=Get l_tempColumns(), Key=Asset Name, Value:"String (by ref)"=Get l_assetName())
   n18 = Break AVT Error Struct(AVT_Error_Struct=Get l_error())
   > Add(TargetMap=Get l_tempColumns(), Key=@n18.csvHeader_49_#15, Value=@n18)
   > Add(
    TargetMap:Map of Strings to AES CSVOutput Struct Structures=Get l_AssetColumnMap()
    Key=Get l_assetName()
    Value:"AES CSVOutput Struct Structure (by ref)"=Make AES CSVOutput Struct!(
     rowName_6_#17:String=Get l_assetName()
     columns_9_#18:Map of Strings to Strings=Get l_tempColumns()
     tags_17_#19:Array of Strings
     vertcount_14_#20:int=0
    ) -> (AES_CSVOutput_Struct:AES CSVOutput Struct Structure) {is=title K2Node_MakeStruct}
   )
   Add(execute=@n17.LoopBody, TargetMap:Map of Strings to Strings=Get l_tempColumns(), Key=@n17, Value:"String (by ref)"=N/A)
   n19 = To Array(execute=@n1.then_3, A=Get l_headers())
   > n20 = For Each Loop(Array=@n19)
   n21 = Sequence!(execute=@n20.Completed) -> (then_0, then_1)
   > Set l_csvRow(l_csvRow:String=Append!(A=Get l_csvRow() -> (l_csvRow:String), B=Append(A=",", B=Vert Count LOD0)) -> (ReturnValue)) -> (Output_Get:String)
   PrintString(execute=@n21.then_1, InString=Get l_csvRow())
   Set l_csvRow(execute=@n20.LoopBody, l_csvRow=Append(A=Get l_csvRow(), B=Append(A=",", B=@n20)))
   n22 = Values(execute=@n1.then_4, TargetMap:Map of Strings to AES CSVOutput Struct Structures=Get l_AssetColumnMap()) -> (Values:Array of AES CSVOutput Struct Structures)
   > n23 = For Each Loop(Array:Array of AES CSVOutput Struct Structures=@n22) -> (Array Element:AES CSVOutput Struct Structure)
   n24 = Break AES CSVOutput Struct(AES_CSVOutput_Struct=@n23)
   > Set l_assetName(l_assetName=@n24)
   > Set l_csvRow()
   > n25 = Values(TargetMap:Map of Strings to Strings=@n24.columns_9_#18) -> (Values:Array of Strings)
   > n26 = For Each Loop(Array:Array of Strings=@n25) -> (Array Element:String)
   n27 = Sequence(execute=@n26.Completed)
   > n28 = getAssetVertexCount!(execute, self, assetName:String=Get l_assetName()) -> (then, vertCount:String)
   > Set l_csvRow(l_csvRow=Append(A=Get l_csvRow(), B=Append(A=",", B=@n28)))
   PrintString(execute=@n27.then_1, InString=Get l_csvRow())
   Set l_csvRow(execute=@n26.LoopBody, l_csvRow=Append(A=Get l_csvRow(), B=Append(A=",", B=@n26)))
  ]}
 SaveGame_PostLoad {nodes [
   n1 = SaveGame_PostLoad!() -> (then, saveGameExisted:bool, saveGame_Object:Save Game Object Reference) {is=title K2Node_FunctionEntry}
   > n2 = Parent: SaveGame_PostLoad!(execute, saveGameExisted:bool=@n1, saveGame_Object:Save Game Object Reference=@n1.saveGame_Object) -> (then, " ":bool) {is=title K2Node_CallParentFunction}
   > Return Node!(execute, " ":bool=@n2)
  ]}
 SaveGame_PrepareToSave {nodes [
   SaveGame_PrepareToSave!() -> (then, SaveGameObject:Save Game Object Reference) {is=title K2Node_FunctionEntry}
   > n1 = Cast To AES_SaveGame(Object=Get saveGame_Object())
   > Set skippedAssetTestMap!(
    execute
    skippedAssetTestMap:Map of Names to AES Skip Test Struct Structures=Get currentSkippedAssetTestMap()
    self:AES Save Game Object Reference=@n1
   ) -> (then, Output_Get:Map of Names to AES Skip Test Struct Structures)
   Return Node(" "=false)
  ]}
 cancelScan {nodes [
   cancelScan() {is=title K2Node_FunctionEntry}
   > setScanMode(isScanning=false)
   > Is Valid(InputObject=Get ScanTask())
   > cancelScan(self:Asset Scan For Error Task Object Reference=Get ScanTask()) {is=title K2Node_CallFunction}
   > Set ScanTask()
  ]}
 getAssetVertexCount {nodes [
   n1 = getAssetVertexCount!() -> (then, assetName:String) {is=title K2Node_FunctionEntry}
   > n2 = Sequence()
   > n3 = GetAssetsByPackageName!(
    execute
    self:Asset Registry Interface=GetAssetRegistry()
    PackageName:Name=String To Name(InString=@n1)
    bIncludeOnlyOnDiskAssets:bool=false
    bSkipARFilteredAssets:bool=true
   ) -> (then, OutAssetData:Array of Asset Data Structures, ReturnValue:bool)
   > Branch(Condition=Is Valid Index(TargetArray:Array of Asset Data Structures=@n3))
   n4 = GetTagValue(
    InAssetData="Get (a copy)"(Array:Array of Asset Data Structures=@n3) -> (Output:Asset Data Structure)
    InTagName=MakeLiteralName(Value=Vertices)
   )
   > Return Node!(execute, vertCount:String=SelectString(A=@n4, B=N/A, bPickA=@n4.ReturnValue))
   Return Node(execute=@n2.then_1, vertCount=N/A)
  ]}
 getErrorCount {nodes [
   getErrorCount() {is=title K2Node_FunctionEntry}
   > n1 = getListItems!(execute, self:TA Generic 01 List Object Reference=Get TA_Generic01_List()) -> (then, items:Array of User Widget Object References)
   n2 = Length(TargetArray:Array of User Widget Object References=@n1)
   > Return Node!(execute, count:int=@n2, hasErrors:bool="integer > integer"(A=@n2))
  ]}
 handleItemErrorHelpClicked {nodes [
   n1 = handleItemErrorHelpClicked!() -> (then, helpCommmand:String) {is=title K2Node_FunctionEntry}
   > Execute Python Script(PythonScript=@n1)
  ]}
 onScanModeChanged {nodes [onScanModeChanged!() -> (then, isScanning:bool=~) {is=title K2Node_FunctionEntry}]}
 setScanMode {nodes [
   n1 = setScanMode!() -> (then, isScanning:bool=~) {is=title K2Node_FunctionEntry}
   > Branch(Condition="Not Equal (Boolean)"(A=@n1, B=Get isScanning()) -> (ReturnValue=~))
   > Set isScanning(isScanning:bool=@n1) -> (Output_Get:bool)
   > n2 = Branch(Condition=Get isScanning())
   > setProgress(self=Get AES_ProgressWidget(), progressString=Starting Scan, percent=0)
   > n3 = SetVisibility(self=Get AES_ProgressWidget(), InVisibility=Visible)
   Call On Scan Mode Changed!(execute:Exec=[@n4.then, @n3.then], self:Asset Error List EWidget Object Reference, isScanning:bool=Get isScanning()) -> (then:Exec) {is=title K2Node_CallDelegate}
   n4 = SetVisibility(execute=@n2.else, self=Get AES_ProgressWidget(), InVisibility=Hidden)
  ]}
]
variables [
 currentSkippedAssetTestMap {type="map<AES_SkipTest_Struct>", default {}}
 isScanning {type=boolean}
 onScanModeChanged {default=()}
 ScanTask {type=AssetScanForError_Task_C, default=null}
]
widgets {
 info {
  parentClass=/Game/Blueprints/Blutilities/ForTechArt/Widgets/TA_Base_EditorWidget.TA_Base_EditorWidget_C
  rootWidgetClass=&Overlay
  widgetCount=3
 }
 widgets [
  Overlay_0 {widgetClassPath=&Overlay, widget=Overlay_0}
  TA_Generic01_List {
   widgetClassPath=/Game/Blueprints/Blutilities/ForTechArt/Widgets/List/TA_Generic01_List
   bIsVariable=true
   parent=Overlay_0
   widget=TA_Generic01_List
   properties {
    allowMultiSelect=false
    itemPadding=0
    listItemClass=/Game/Blueprints/Blutilities/ForArtists/AssetValidationTool/SubWidgets/AVT_Asset_ListItemWidget.AVT_Asset_ListItemWidget_C
   }
   slot {class=OverlaySlot, properties {horizontalAlignment=HAlign_Fill, verticalAlignment=VAlign_Fill}}
  }
  AES_ProgressWidget {
   widgetClassPath=&/AES_ProgressWidget
   bIsVariable=true
   parent=Overlay_0
   widget=AES_ProgressWidget
   properties {visibility=Hidden}
   slot {class=OverlaySlot, properties {horizontalAlignment=HAlign_Fill, verticalAlignment=VAlign_Fill}}
  }
 ]
}
