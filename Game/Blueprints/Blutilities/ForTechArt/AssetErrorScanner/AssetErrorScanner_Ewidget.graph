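# Tech-art editor utility widget that scans selected Content Browser assets for static mesh errors.
# Run pairs each selection with vertex count, collision, shadow and two LOD task classes, caching the list for reruns.
# AssetErrorList_EWidget runs the tests after a time-estimate confirmation for batches over 100, and prints CSV to the log.
# Its scan-mode event refreshes button states; Help opens a Google Doc through Python.
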EditorUtilityWidgetBlueprint /Game/Blueprints/Blutilities/ForTechArt/AssetErrorScanner/AssetErrorScanner_Ewidget
graphs [
 EventGraph {nodes [
   "On Clicked (btnRun)"!() -> (OutputDelegate, then, widget:User Widget Object Reference, event:Pointer Event Structure) {is=title K2Node_ComponentBoundEvent}
   > RunTestsOnSelectedAssets()
   "On Scan Mode Changed (AssetErrorList_EWidget)"!() -> (OutputDelegate, then, isScanning:bool) {is=title K2Node_ComponentBoundEvent}
   > validateButtonVisuals()
   "On Clicked (btnCSV)"!() -> (OutputDelegate, then, widget:User Widget Object Reference, event:Pointer Event Structure) {is=title K2Node_ComponentBoundEvent}
   > printAssetsToCsv()
   "On Clicked (btnHelp)"!() -> (OutputDelegate, then, widget:User Widget Object Reference, event:Pointer Event Structure) {is=title K2Node_ComponentBoundEvent}
   > ExecutePythonCommand(PythonCommand="import webbrowser\r\nwebbrowser.open('https://docs.google.com/document/d/1wEujfhk_3q1a7qNpzje9hpiVEaAhUBbaB0TqCgGFeJw/edit?usp=sharing', new=2)")
   "On Clicked (btnRefresh)"!() -> (OutputDelegate, then, widget:User Widget Object Reference, event:Pointer Event Structure) {is=title K2Node_ComponentBoundEvent}
   > ReRunTests()
   Event Construct()
   > Parent: Construct()
   > validateButtonVisuals()
  ]}
 ReRunTests {nodes [
   ReRunTests() {is=title K2Node_FunctionEntry}
   > showPromptAndRunTests()
  ]}
 RunTestsOnSelectedAssets {nodes [
   RunTestsOnSelectedAssets() {is=title K2Node_FunctionEntry}
   > n1 = gatherAssetForTests!(execute, self) -> (then, Tests:Array of AES Tests Struct Structures)
   > Set TestsToRun(TestsToRun:Array of AES Tests Struct Structures=@n1) -> (Output_Get:Array of AES Tests Struct Structures)
   > showPromptAndRunTests()
  ]}
 gatherAssetForTests {nodes [
   gatherAssetForTests() {is=title K2Node_FunctionEntry}
   > n1 = GetSelectedAssetData()
   > n2 = For Each Loop(Array:Array of Asset Data Structures=@n1) -> (Array Element:Asset Data Structure)
   Return Node!(execute=@n2.Completed, Tests:Array of AES Tests Struct Structures=Get l_TestsToRun() -> (l_TestsToRun:Array of AES Tests Struct Structures))
   > Add(
    TargetArray:Array of AES Tests Struct Structures=Get l_TestsToRun()
    NewItem:"AES Tests Struct Structure (by ref)"=Make AES Tests Struct!(
     assetData_9_#1:Asset Data Structure=@n2
     asset_8_#2:Object Reference
     tests_10_#3:Array of Base Error Task Class References=Make Array!(
      "[0]":Base Error Task Class Reference=/Game/Blueprints/Blutilities/ForTechArt/AssetErrorScanner/Tasks/Error_SM_VertCount_Task.Error_SM_VertCount_Task_C
      "[1]":Base Error Task Class Reference=/Game/Blueprints/Blutilities/ForTechArt/AssetErrorScanner/Tasks/Error_SM_Collision_Task.Error_SM_Collision_Task_C
      "[2]":Base Error Task Class Reference=/Game/Blueprints/Blutilities/ForTechArt/AssetErrorScanner/Tasks/Error_SM_Shadows_Task.Error_SM_Shadows_Task_C
      "[3]":Base Error Task Class Reference=/Game/Blueprints/Blutilities/ForTechArt/AssetErrorScanner/Tasks/Error_SM_LODSimilar_Task.Error_SM_LODSimilar_Task_C
      "[4]":Base Error Task Class Reference=/Game/Blueprints/Blutilities/ForTechArt/AssetErrorScanner/Tasks/Error_SM_NoLOD_Task.Error_SM_NoLOD_Task_C
     ) -> (Array:Array of Base Error Task Class References)
    ) -> (AES_Tests_Struct:AES Tests Struct Structure) {is=title K2Node_MakeStruct}
   ) {is=title K2Node_CallArrayFunction}
  ]}
 printAssetsToCsv {nodes [
   printAssetsToCsv() {is=title K2Node_FunctionEntry}
   > PrintErrorDataCsv(self:Asset Error List EWidget Object Reference=Get AssetErrorList_EWidget() -> (AssetErrorList_EWidget:Asset Error List EWidget Object Reference))
  ]}
 showPromptAndRunTests {nodes [
   showPromptAndRunTests() {is=title K2Node_FunctionEntry}
   n1 = Get TestsToRun() -> (TestsToRun:Array of AES Tests Struct Structures)
   > n2 = Branch(Condition="integer > integer"(A=Length(TargetArray:Array of AES Tests Struct Structures=@n1), B=100))
   n3 = Length(TargetArray=Get TestsToRun())
   n4 = "To String (Integer)"(InInt=@n3)
   n5 = float * float(A:int=@n3, B=1.2) {is=title K2Node_PromotableOperator}
   n6 = Ceil(A=@n5) {is=title K2Node_CallFunction}
   n7 = "To Text (Float)"(Value=float / float(A=@n5, B=60), MinimumIntegralDigits=1, MaximumIntegralDigits=324, MinimumFractionalDigits=0, MaximumFractionalDigits=1)
   n8 = SelectString(
    A=Append(A="To String (Text)"(InText=@n7), B=" minutes")
    B=Append(A="To String (Integer)"(InInt=@n6), B=" seconds")
    bPickA="float > float"(A=@n5, B=59)
   )
   > n9 = Show Message Dialog(
    Title=MakeLiteralText(Value="LOCTEXT(\"Run Asset Scanner\")")
    Message="To Text (String)"(InString=Append!(
      A="You are attempting to run the Asset Error Scanner on "
      B=@n4
      C:String=" assets.  This could take a good bit of time.  (About "
      D:String=@n8
      E:String=").\r\n\r\nDo you want to continue?"
     ) -> (ReturnValue))
    MessageType=OkCancel
    DefaultValue=Cancel
    MessageCategory=Warning
   )
   > n10 = Branch(Condition="Equal (Enum)"(A:EAppReturnType Enum=@n9, B:EAppReturnType Enum=Ok))
   RunTests!(
    execute=[@n2.else, @n10.then]
    self:Asset Error List EWidget Object Reference=Get AssetErrorList_EWidget()
    assetTests:Array of AES Tests Struct Structures=Get TestsToRun()
   ) -> (then)
  ]}
 validateButtonVisuals {nodes [
   validateButtonVisuals() {is=title K2Node_FunctionEntry}
   n1 = Get isScanning!(self:Asset Error List EWidget Object Reference=Get AssetErrorList_EWidget()) -> (isScanning:bool)
   > Set buttonColor_Default!(
    execute
    buttonColor_Default:Linear Color Structure=SelectColor(A=(R=0.276,G=0.0556,B=0.017,A=1), B=(R=0.102,G=0.451,B=1,A=1), bPickA=@n1)
    self:TA Base Editor Button Widget Object Reference=Get btnRun() -> (btnRun:TA Generic 01 Button Object Reference)
   ) -> (then, Output_Get:Linear Color Structure)
   > setLabel!(
    execute
    self:TA Base Editor Button Widget Object Reference=Get btnRun()
    label:String=SelectString(A=Scanning in progress..., B=Click to scan selected assets., bPickA=@n1)
   ) -> (then)
   n2 = NOT Boolean(A=@n1)
   > SetIsEnabled(self=[Get btnRun(), Get btnCSV() -> (btnCSV:TA Generic 01 Button Object Reference)], bInIsEnabled=@n2)
   n3 = Length(TargetArray=Get TestsToRun())
   > SetIsEnabled(self=Get btnRefresh() -> (btnRefresh:TA Generic 01 Button Object Reference), bInIsEnabled=AND Boolean(A=@n2, B="integer > integer"(A=@n3)))
   n4 = Get AssetErrorList_EWidget()
   > SetIsEnabled(
    self=Get btnCSV()
    bInIsEnabled=AND Boolean(A=@n2, B=getErrorCount!(self:Asset Error List EWidget Object Reference=@n4) -> (count:int=0, hasErrors:bool).hasErrors)
   )
  ]}
]
variables [
 ScanTask {type=AssetScanForError_Task_C, default=null}
 TestsToRun {type="array<AES_Tests_Struct>", default []}
]
widgets {
 info {
  parentClass=/Game/Blueprints/Blutilities/ForTechArt/Widgets/TA_Base_EditorWidget.TA_Base_EditorWidget_C
  rootWidgetClass=&Overlay
  widgetCount=8
 }
 widgets [
  Overlay_0 {widgetClassPath=&Overlay, widget=Overlay_0}
  VerticalBox_0 {
   widgetClassPath=&VerticalBox
   parent=Overlay_0
   widget=VerticalBox_0
   slot {class=OverlaySlot, properties {horizontalAlignment=HAlign_Fill, verticalAlignment=VAlign_Fill}}
  }
  HorizontalBox_0 {widgetClassPath=&HorizontalBox, parent=VerticalBox_0, widget=HorizontalBox_0, slot {class=VerticalBoxSlot}}
  btnRun {
   widgetClassPath=&TA_Generic01_Button
   parent=HorizontalBox_0
   widget=btnRun
   properties {label=Run Tests On Selected Assets In Content Browser}
   slot {class=HorizontalBoxSlot, properties {padding=<ltrb 5 0 0 0>, size {sizeRule=Fill, value=1}}}
  }
  btnRefresh {
   widgetClassPath=&TA_Generic01_Button
   parent=HorizontalBox_0
   widget=btnRefresh
   properties {bIsEnabled=false, label=Rerun Tests, buttonColor_Default=<rgba 0.2041 0.4271 0.1008 1>}
   slot {class=HorizontalBoxSlot, properties {padding=<ltrb 5 0 0 0>, size {sizeRule=Fill, value=0.25}}}
  }
  btnHelp {
   widgetClassPath=&TA_Generic01_Button
   parent=HorizontalBox_0
   widget=btnHelp
   properties {label=Help, buttonColor_Default=<rgba 0.5885 0.3192 0.1005 1>}
   slot {class=HorizontalBoxSlot, properties {padding=<ltrb 5 0 5 0>}}
  }
  AssetErrorList_EWidget {
   widgetClassPath=/Game/Blueprints/Blutilities/ForTechArt/AssetErrorScanner/Widgets/AssetErrorList_EWidget
   bIsVariable=true
   parent=VerticalBox_0
   widget=AssetErrorList_EWidget
   slot {class=VerticalBoxSlot, properties {size {sizeRule=Fill, value=1}}}
  }
  btnCSV {
   widgetClassPath=&TA_Generic01_Button
   parent=VerticalBox_0
   widget=btnCSV
   properties {label=Print CSV to Log, buttonColor_Default=<rgba 0.1933 0.6406 0.154 1>}
   slot {class=VerticalBoxSlot}
  }
 ]
}
